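# Looping ambient SoundCue for Fortilla props that plays metal-on-water knock and splash sounds.
# A Looping node repeats a randomized Delay of a few seconds, feeding a Modulator that slightly varies pitch and volume.
# A Random node picks among twelve equally weighted splash waves on each pass.
# It uses the water ambient sound class, with Fortilla splash attenuation and concurrency assets.
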SoundCue /Game/Sounds/Apollo/Props/Fortillas/Amb_Fortilla_Splash_Looping_Cue
root=Looping_0
nodes [
 {id=Delay_0, children [Modulator_0], properties {delayMax=6, delayMin=1}}
 {id=Looping_0, children [Delay_0]}
 {id=Modulator_0, children [Random_0], properties {pitchMax=1.1, pitchMin=0.9, volumeMax=1.1, volumeMin=0.98}}
 {
  id=Random_0
  children [
   WavePlayer_0
   WavePlayer_1
   WavePlayer_4
   WavePlayer_5
   WavePlayer_9
   WavePlayer_10
   WavePlayer_11
   WavePlayer_12
   WavePlayer_13
   WavePlayer_14
   WavePlayer_15
   WavePlayer_16
  ]
  properties {weights [1,1,1,1,1,1,1,1,1,1,1,1]}
 }
 {id=WavePlayer_0, properties {soundWaveAssetPtr=&/Amb_Metal_Water_Floatilla_Knock_Splash_03_10}}
 {id=WavePlayer_1, properties {soundWaveAssetPtr=&/Amb_Metal_Water_Floatilla_Knock_Splash_03_11}}
 {id=WavePlayer_10, properties {soundWaveAssetPtr=&/Amb_Metal_Water_Floatilla_Knock_Splash_03_03}}
 {id=WavePlayer_11, properties {soundWaveAssetPtr=&/Amb_Metal_Water_Floatilla_Knock_Splash_03_04}}
 {id=WavePlayer_12, properties {soundWaveAssetPtr=&/Amb_Metal_Water_Floatilla_Knock_Splash_03_05}}
 {id=WavePlayer_13, properties {soundWaveAssetPtr=&/Amb_Metal_Water_Floatilla_Knock_Splash_03_06}}
 {id=WavePlayer_14, properties {soundWaveAssetPtr=&/Amb_Metal_Water_Floatilla_Knock_Splash_03_07}}
 {id=WavePlayer_15, properties {soundWaveAssetPtr=&/Amb_Metal_Water_Floatilla_Knock_Splash_03_08}}
 {id=WavePlayer_16, properties {soundWaveAssetPtr=&/Amb_Metal_Water_Floatilla_Knock_Splash_03_09}}
 {id=WavePlayer_4, properties {soundWaveAssetPtr=&/Amb_Metal_Water_Floatilla_Knock_Splash_03_12}}
 {id=WavePlayer_5, properties {soundWaveAssetPtr=&/Amb_Metal_Water_Floatilla_Knock_Splash_03_01}}
 {id=WavePlayer_9, properties {soundWaveAssetPtr=&/Amb_Metal_Water_Floatilla_Knock_Splash_03_02}}
]
properties {
 attenuationSettings=&/Fortilla_Splash_Att
 maxDistance=2512
 soundClassObject=/Game/Sounds/FortSoundClassesAndModes/Fort_Water_Ambient
 volumeMultiplier=0.95
 attenuationOverrides {
  bAttenuateWithLPF=true
  falloffDistance=6000
  lPFFrequencyAtMax=500
  nonSpatializedRadiusStart=700
  stereoSpread=1000
  attenuationShapeExtents {x=925}
 }
 concurrencyOverrides {maxCount=1, resolutionRule=PreventNew}
 concurrencySet [&/Fortilla_Splash_Conc]
}
